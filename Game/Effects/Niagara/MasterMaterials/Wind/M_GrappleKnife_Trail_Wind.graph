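# Translucent unlit, two-sided Niagara ribbon material for the wind trail of the grapple knife effect.
# Particle color drives emissive, while opacity sums two blended, panning smoke noise samples with a slash-trail streak layer, both shaped by UV edge masks.
# The streak's rotated UVs lerp between two tilings along the ribbon, with a switchable curved sphere mask.
# Particle alpha and depth fade scale the result, with the fade skipped on mobile.
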Material /Game/Effects/Niagara/MasterMaterials/Wind/M_GrappleKnife_Trail_Wind
graph {
 properties {
  bUsedWithNiagaraRibbons=true
  blendMode=BLEND_Translucent
  pixelDepthOffsetMode=PDOM_Legacy
  shadingModel=MSM_Unlit
  twoSided=true
  usedShadingModels=Unlit
 }
 nodes [
  n1 = ParticleColor()
  MaterialFunctionCall(Emissive=@n1.RGB) {materialFunction=&MF_FX_EmissiveAdjustments}
  n2 = TextureCoordinate()
  n3 = VectorParameter() {parameterName=UV_Tile, channelNames {a=V_Speed, b=U_Speed, g=V_Scale, r=U_Scale}, defaultValue=<rgba 0.25 1 0.1789 -0.05>}
  n4 = AppendVector(A=@n3.U_Scale, B=@n3.U_Scale)
  n5 = AppendVector(A=@n3.U_Speed, B=@n3.U_Speed)
  n6 = Time()
  n7 = TextureSampleParameter2D(UVs=Add(A=Multiply(A=@n2, B=@n4), B=Multiply(A=@n5, B=@n6))) {parameterName=NoiseA, texture=&T_Inky_Smoke_Tile_Chromatic}
  n8 = VectorParameter() {parameterName=UV_01_Tile, channelNames {a=V_Speed, b=U_Speed, g=V_Scale, r=U_Scale}, defaultValue=<rgba 0.2 1 0.275 0.1325>}
  n9 = AppendVector(A=@n8.U_Scale, B=@n8.U_Scale)
  n10 = AppendVector(A=@n8.U_Speed, B=@n8.U_Speed)
  n11 = TextureSampleParameter2D(UVs=Add(A=Multiply(A=@n2, B=@n9), B=Multiply(A=@n10, B=@n6))) {parameterName=NoiseB, texture=&T_Inky_Smoke_Tile_Chromatic}
  n12 = TextureCoordinate()
  n13 = OneMinus(None=Saturate(None=ComponentMask(None=@n12) {r=true}))
  n14 = Multiply(A=ComponentMask(None=TextureCoordinate()) {r=true}) {constB=4}
  n15 = ComponentMask(None=TextureCoordinate()) {g=true}
  n16 = Saturate(None=Multiply(A=@n15) {constB=2})
  n17 = Saturate(None=Multiply(A=OneMinus(None=@n15)) {constB=2})
  n18 = Multiply(A=Multiply(A=Power(Base=@n13) {constExponent=2.5}, B=Saturate(None=@n14)), B=Power(Base=Multiply(A=@n16, B=@n17)) {constExponent=1})
  n19 = Multiply(A=Multiply(A=LinearInterpolate(A=@n7.G, B=@n11.G), B=Saturate(None=@n18))) {constB=0.5}
  n20 = Constant() {r=2}
  n21 = ScalarParameter() {defaultValue=1, parameterName=RightExp}
  n22 = ScalarParameter() {defaultValue=2, parameterName=TopBotMult}
  n23 = ScalarParameter() {defaultValue=1, parameterName=TopBotExp}
  n24 = Multiply(
   A=Multiply(
    A=MaterialFunctionCall!(UVs, Mult=@n20, Exp=@n21) {materialFunction=/Game/Effects/Fort_Effects/MaterialFunctions/MF_SquareEdgeMask}.RightMask
    B=MaterialFunctionCall(Mult=@n22, Exp=@n23) {materialFunction=&MF_SquareEdgeMask}.HorizMask
   )
   B=MaterialFunctionCall(Mult=Constant() {r=4}, Exp=Constant() {r=1.5}) {materialFunction=&MF_SquareEdgeMask}.LeftMask
  )
  n25 = MaterialFunctionCall(Float2=TextureCoordinate() {uTiling=1.2, vTiling=1.5}) {materialFunction=&BreakOutFloat2Components}
  n26 = MaterialFunctionCall(X=OneMinus(None=@n25.G), Y=Add(A=@n25.R) {constA=0.25, constB=-0.3}) {materialFunction=&MakeFloat2}
  n27 = MaterialFunctionCall(Float2=TextureCoordinate() {uTiling=12, vTiling=3}) {materialFunction=&BreakOutFloat2Components}
  n28 = MaterialFunctionCall(X=OneMinus(None=@n27.G), Y=Add(B=@n27.R) {constA=0.25}) {materialFunction=&MakeFloat2}
  n29 = MaterialFunctionCall(Float2=TextureCoordinate()) {materialFunction=&BreakOutFloat2Components}
  n30 = Multiply(
   A=TextureSample(UVs=LinearInterpolate(A=@n26.Result, B=@n28.Result, Alpha=@n29.G)) {samplerType=SAMPLERTYPE_LinearColor, texture=&/T_Slash_Trail}.R
   B=Power(Base=OneMinus(None=@n29.G))
  )
  n31 = OneMinus(None=ComponentMask(None=TextureCoordinate()) {r=true})
  n32 = OneMinus(None=MaterialFunctionCall(Mult=Constant() {r=10}, Exp=Constant() {r=0.8}) {materialFunction=&MF_SquareEdgeMask}.LeftMask)
  n33 = OneMinus(None=SphereMask(A=TextureCoordinate() {uTiling=0.5}, B=Constant2Vector() {g=0.82, r=0.85}, Radius=Constant() {r=0.8}, Hardness=Constant() {r=0.3}))
  n34 = Multiply(
   A=Add(A=Add(A=@n31, B=@n32), B=@n30)
   B=StaticSwitchParameter(True=Constant() {r=1}, False=Power(Base=@n33) {constExponent=5}) {parameterName=TurnOffCurvedMask}
  )
  n35 = Multiply(B=Add(A=Multiply(A=@n30, B=@n24) {constB=0.25}, B=Multiply(A=@n34, B=@n24))) {constA=0.5}
  n36 = DepthFade(FadeDistance=ScalarParameter() {defaultValue=20, parameterName=FadeDistance})
  Multiply(
   A=Multiply(A=@n1.A, B=Add(A=@n19, B=@n35))
   B=ShadingPathSwitch(Default=@n36, Deferred=@n36, Mobile=Constant() {r=1}) {desc=Migrated from FeatureLevelSwitch}
  )
 ]
}
