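# Translucent, unlit, two-sided master material for wind effects on Niagara mesh particles or instanced static meshes.
# Sums four tiled, power-adjusted wind textures, then masks them with optional UV edge gradients.
# UseInNiagara switches color, opacity and offsets from particle color and dynamic parameters to material parameters, panners and a color gradient.
# Opacity also gets depth, view-angle and camera-distance fades, plus optional axis-aligned fresnel; emissive runs through MF_FX_EmissiveAdjustments.
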Material /Game/Effects/Niagara/MasterMaterials/Wind/M_Mesh_Wind
graph {
 properties {
  bUsedWithInstancedStaticMeshes=true
  bUsedWithNiagaraMeshParticles=true
  blendMode=BLEND_Translucent
  pixelDepthOffsetMode=PDOM_Legacy
  shadingModel=MSM_Unlit
  twoSided=true
  usedShadingModels=Unlit
 }
 nodes [
  n1 = ParticleColor()
  n2 = VectorParameter() {parameterName=Color, defaultValue=<rgba 1 1 1 1>}
  n3 = VectorParameter() {parameterName=ColorBase, defaultValue=<rgba 1 1 1 1>}
  n4 = ComponentMask(None=TextureCoordinate()) {g=true}
  MaterialFunctionCall(Emissive=StaticSwitchParameter(True=@n1.RGB, False=LinearInterpolate(A=@n2.RGB, B=@n3.RGB, Alpha=@n4)) {defaultValue=true, parameterName=UseInNiagara?}) {materialFunction=&MF_FX_EmissiveAdjustments}
  n5 = DynamicParameter() {parameterIndex=1, paramNames [Wind 01 Vertical Offset, Wind 01 Horizontal Offset, Wind 02 Vertical Offset, Wind 02 Horizontal Offset]}
  n6 = ComponentMask(None=StaticSwitchParameter(True=@n5.Wind 01 Vertical Offset, False=ScalarParameter() {parameterName=Wind01_VerticalOffset}) {defaultValue=true, parameterName=UseInNiagara?}) {g=true}
  n7 = ComponentMask(None=StaticSwitchParameter(True=@n5.Wind 01 Horizontal Offset, False=ScalarParameter() {parameterName=Wind01_HorisontalOffset}) {defaultValue=true, parameterName=UseInNiagara?}) {r=true}
  n8 = TextureCoordinate()
  n9 = Multiply(
   A=ScalarParameter() {defaultValue=1, group="1.1) Wind Texture Tiling", parameterName=Wind 01 Vertical Tiling, sortPriority=0}
   B=ComponentMask(None=@n8) {g=true}
  )
  n10 = Multiply(
   A=ScalarParameter() {defaultValue=1, group="1.1) Wind Texture Tiling", parameterName=Wind 01 Horizontal Tiling, sortPriority=1}
   B=ComponentMask(None=@n8) {r=true}
  )
  NamedRerouteDeclaration(None=Add(A=AppendVector(A=@n6, B=@n7), B=AppendVector(A=@n9, B=@n10))) {name=Wind Texture 01 Tiling and Offset, nodeColor=<rgba 1 0.1406 0 1>}
  n11 = DynamicParameter() {parameterIndex=1, paramNames [Wind 01 Vertical Offset, Wind 01 Horizontal Offset, Wind 02 Vertical Offset, Wind 02 Horizontal Offset]}
  n12 = ComponentMask(None=StaticSwitchParameter(True=@n11.Wind 02 Vertical Offset, False=ScalarParameter() {parameterName=Wind 02 Vertical Offset}) {defaultValue=true, parameterName=UseInNiagara?}) {g=true}
  n13 = ComponentMask(None=StaticSwitchParameter(True=@n11.Wind 02 Horizontal Offset, False=ScalarParameter() {parameterName=Wind 02 Horizontal Offset}) {defaultValue=true, parameterName=UseInNiagara?}) {r=true}
  n14 = TextureCoordinate()
  n15 = Multiply(
   A=ScalarParameter() {defaultValue=1, group="1.1) Wind Texture Tiling", parameterName=Wind 02 Vertical Tiling, sortPriority=2}
   B=ComponentMask(None=@n14) {g=true}
  )
  n16 = Multiply(
   A=ScalarParameter() {defaultValue=1, group="1.1) Wind Texture Tiling", parameterName=Wind 02 Horizontal Tiling, sortPriority=3}
   B=ComponentMask(None=@n14) {r=true}
  )
  NamedRerouteDeclaration(None=Add(A=AppendVector(A=@n12, B=@n13), B=AppendVector(A=@n15, B=@n16))) {name=Wind Texture 02 Tiling and Offset, nodeColor=<rgba 0 1 0.9297 1>}
  n17 = DynamicParameter() {parameterIndex=2, paramNames [Wind 03 Vertical Offset, Wind 03 Horizontal Offset, Wind 04 Vertical Offset, Wind 04 Horizontal Offset]}
  n18 = ComponentMask(None=StaticSwitchParameter(True=@n17.Wind 03 Vertical Offset, False=ScalarParameter() {parameterName=Wind 03 Vertical Offset}) {defaultValue=true, parameterName=UseInNiagara?}) {g=true}
  n19 = ComponentMask(None=StaticSwitchParameter(True=@n17.Wind 03 Horizontal Offset, False=ScalarParameter() {parameterName=Wind 03 Horizontal Offset}) {defaultValue=true, parameterName=UseInNiagara?}) {r=true}
  n20 = TextureCoordinate()
  n21 = Multiply(
   A=ScalarParameter() {defaultValue=1, group="1.1) Wind Texture Tiling", parameterName=Wind 03 Vertical Tiling, sortPriority=4}
   B=ComponentMask(None=@n20) {g=true}
  )
  n22 = Multiply(
   A=ScalarParameter() {defaultValue=1, group="1.1) Wind Texture Tiling", parameterName=Wind 03 Horizontal Tiling, sortPriority=5}
   B=ComponentMask(None=@n20) {r=true}
  )
  NamedRerouteDeclaration(None=Add(A=AppendVector(A=@n18, B=@n19), B=AppendVector(A=@n21, B=@n22))) {name=Wind Texture 03 Tiling and Offset, nodeColor=<rgba 0.8516 0 1 1>}
  n23 = DynamicParameter() {parameterIndex=2, paramNames [Wind 03 Vertical Offset, Wind 03 Horizontal Offset, Wind 04 Vertical Offset, Wind 04 Horizontal Offset]}
  n24 = ComponentMask(None=StaticSwitchParameter(True=@n23.Wind 04 Vertical Offset, False=ScalarParameter() {parameterName=Wind 04 Vertical Offset}) {defaultValue=true, parameterName=UseInNiagara?}) {g=true}
  n25 = ComponentMask(None=StaticSwitchParameter(True=@n23.Wind 04 Horizontal Offset, False=ScalarParameter() {parameterName=Wind 04 Horizontal Offset}) {defaultValue=true, parameterName=UseInNiagara?}) {r=true}
  n26 = TextureCoordinate()
  n27 = Multiply(
   A=ScalarParameter() {defaultValue=1, group="1.1) Wind Texture Tiling", parameterName=Wind 04 Vertical Tiling, sortPriority=6}
   B=ComponentMask(None=@n26) {g=true}
  )
  n28 = Multiply(
   A=ScalarParameter() {defaultValue=1, group="1.1) Wind Texture Tiling", parameterName=Wind 04 Horizontal Tiling, sortPriority=7}
   B=ComponentMask(None=@n26) {r=true}
  )
  NamedRerouteDeclaration(None=Add(A=AppendVector(A=@n24, B=@n25), B=AppendVector(A=@n27, B=@n28))) {name=Wind Texture 04 Tiling and Offset, nodeColor=<rgba 1 0 0.9844 1>}
  n29 = NamedRerouteUsage()
  n30 = AppendVector(A=ScalarParameter() {parameterName=Wind 01 Speed X}, B=ScalarParameter() {parameterName=Wind 01 Speed Y})
  n31 = TextureSampleParameter2D(UVs=StaticSwitchParameter(True=@n29, False=Panner(Coordinate=@n29, Speed=@n30)) {defaultValue=true, parameterName=UseInNiagara?}) {group="1) Wind Textures", parameterName=Wind Texture 01, sortPriority=0, texture=&T_ShieldCaustics}
  n32 = DynamicParameter() {defaultValue=<rgba 1 0 0 0>, paramNames [Wind 01 Opacity, Wind 02 Opacity, Wind 03 Opacity, Wind 04 Opacity]}
  n33 = StaticSwitchParameter(True=@n32.Wind 01 Opacity, False=ScalarParameter() {defaultValue=1, parameterName=Wind 01 Opacity}) {defaultValue=true, parameterName=UseInNiagara?}
  n34 = NamedRerouteUsage()
  n35 = AppendVector(A=ScalarParameter() {parameterName=Wind 02 Speed X}, B=ScalarParameter() {parameterName=Wind 02 Speed Y})
  n36 = TextureSampleParameter2D(UVs=StaticSwitchParameter(True=@n34, False=Panner(Coordinate=@n34, Speed=@n35)) {defaultValue=true, parameterName=UseInNiagara?}) {
   group="1) Wind Textures"
   parameterName=Wind Texture 02
   samplerType=SAMPLERTYPE_LinearGrayscale
   sortPriority=2
   texture=/Game/Global/Textures/Default/Flowmaps/T_StormSwirlMask
  }
  n37 = StaticSwitchParameter(True=@n32.Wind 02 Opacity, False=ScalarParameter() {defaultValue=1, parameterName=Wind 02 Opacity}) {defaultValue=true, parameterName=UseInNiagara?}
  n38 = Add(
   A=Power(
    Base=Multiply(A=@n31.R, B=@n33)
    Exp=ScalarParameter() {defaultValue=1, group="1) Wind Textures", parameterName=Wind 01 Power, sortPriority=1}
   )
   B=Power(
    Base=Multiply(A=@n36.R, B=@n37)
    Exp=ScalarParameter() {defaultValue=1, group="1) Wind Textures", parameterName=Wind 02 Power, sortPriority=3}
   )
  )
  n39 = NamedRerouteUsage()
  n40 = AppendVector(A=ScalarParameter() {parameterName=Wind 03 Speed X}, B=ScalarParameter() {parameterName=Wind 03 Speed Y})
  n41 = TextureSampleParameter2D(UVs=StaticSwitchParameter(True=@n39, False=Panner(Coordinate=@n39, Speed=@n40)) {defaultValue=true, parameterName=UseInNiagara?}) {group="1) Wind Textures", parameterName=Wind Texture 03, sortPriority=4, texture=/Game/Effects/Fort_Effects/Textures/Tile/T_CloudScape_08}
  n42 = ScalarParameter() {defaultValue=1, parameterName=Wind 03 Opacity}
  n43 = Power(
   Base=Multiply(A=@n41.R, B=StaticSwitchParameter(True=@n32.Wind 03 Opacity, False=@n42) {defaultValue=true, parameterName=UseInNiagara?})
   Exp=ScalarParameter() {defaultValue=2, group="1) Wind Textures", parameterName=Wind 03 Power, sortPriority=5}
  )
  n44 = NamedRerouteUsage()
  n45 = AppendVector(A=ScalarParameter() {parameterName=Wind 04 Speed X}, B=ScalarParameter() {parameterName=Wind 04 Speed Y})
  n46 = TextureSampleParameter2D(UVs=StaticSwitchParameter(True=@n44, False=Panner(Coordinate=@n44, Speed=@n45)) {defaultValue=true, parameterName=UseInNiagara?}) {
   group="1) Wind Textures"
   parameterName=Wind Texture 04
   sortPriority=6
   texture=&T_Inky_Smoke_Tile
  }
  n47 = Multiply(
   A=@n46.R
   B=StaticSwitchParameter(True=@n32.Wind 04 Opacity, False=ScalarParameter() {defaultValue=1, parameterName=Wind 04 Opacity}) {defaultValue=true, parameterName=UseInNiagara?}
  )
  n48 = ScalarParameter() {defaultValue=1, group="1) Wind Textures", parameterName=Wind 04 Power, sortPriority=7}
  NamedRerouteDeclaration(None=Add(A=Add(A=@n38, B=@n43), B=Power(Base=@n47, Exp=@n48))) {name=Wind Textures, nodeColor=<rgba 0 1 0.8125 1>}
  n49 = TextureCoordinate()
  n50 = Multiply(
   A=ComponentMask(None=@n49) {r=true}
   B=ScalarParameter() {defaultValue=1, group="2) Left Gradient", parameterName=Left Gradient Strength, sortPriority=1}
  )
  n51 = ScalarParameter() {defaultValue=1, group="2) Left Gradient", parameterName=Left Gradient Power, sortPriority=2}
  n52 = Multiply(
   A=OneMinus(None=ComponentMask(None=@n49) {r=true})
   B=ScalarParameter() {defaultValue=1, group="2.1) Right Gradient", parameterName=Right Gradient Strength, sortPriority=1}
  )
  n53 = ScalarParameter() {defaultValue=1, group="2.1) Right Gradient", parameterName=Right Gradient Power, sortPriority=2}
  n54 = Multiply(
   A=LinearInterpolate(
    A=Constant() {r=1}
    B=Power(Base=@n50, Exp=@n51)
    Alpha=ScalarParameter() {group="2) Left Gradient", parameterName=Use Left Gradient, sortPriority=0}
   )
   B=LinearInterpolate(
    A=Constant() {r=1}
    B=Power(Base=@n52, Exp=@n53)
    Alpha=ScalarParameter() {group="2.1) Right Gradient", parameterName=Use Right Gradient, sortPriority=0}
   )
  )
  n55 = ComponentMask(None=@n49) {g=true}
  n56 = ScalarParameter() {defaultValue=1, group="2.2) Top Gradient", parameterName=Top Gradient Strength, sortPriority=1}
  n57 = LinearInterpolate(
   A=Constant() {r=1}
   B=Power(
    Base=Multiply(A=@n55, B=@n56)
    Exp=ScalarParameter() {defaultValue=1, group="2.2) Top Gradient", parameterName=Top Gradient Power, sortPriority=2}
   )
   Alpha=ScalarParameter() {group="2.2) Top Gradient", parameterName=Use Top Gradient, sortPriority=0}
  )
  n58 = Constant() {r=1}
  n59 = ComponentMask(None=@n49) {g=true}
  n60 = Power(
   Base=Multiply(
    A=OneMinus(None=@n59)
    B=ScalarParameter() {defaultValue=1, group="2.3) Bottom Gradient", parameterName=Bottom Gradient Strength, sortPriority=1}
   )
   Exp=ScalarParameter() {defaultValue=1, parameterName=Bottom Gradient Power, sortPriority=2}
  )
  n61 = ScalarParameter() {group="2.3) Bottom Gradient", parameterName=Use Bottom Gradient, sortPriority=0}
  NamedRerouteDeclaration(None=Multiply(A=Multiply(A=@n54, B=@n57), B=LinearInterpolate(A=@n58, B=@n60, Alpha=@n61))) {name=Gradient Masks, nodeColor=<rgba 0.875 0 1 1>}
  n62 = VectorParameter() {group=Fresnel, parameterName="Axis (V3)", defaultValue=<rgba 0 0 1 0>}
  n63 = ScalarParameter() {defaultValue=1, group=Fresnel, parameterName="Axis Fresnel Exponent (S)"}
  n64 = Multiply(
   A=StaticSwitchParameter(True=@n1.A, False=ScalarParameter() {defaultValue=1, parameterName=Opacity}) {defaultValue=true, parameterName=UseInNiagara?}
   B=Multiply(A=NamedRerouteUsage(), B=NamedRerouteUsage())
  )
  n65 = ScalarParameter() {defaultValue=10, parameterName=DepthFade}
  n66 = ScalarParameter() {desc=Camera Fade distance., group="4) Camera Fade", parameterName=Camera Angle Fade, sortPriority=4}
  n67 = Multiply(
   A=Multiply(
    A=DepthFade(Opacity=@n64, FadeDistance=@n65)
    B=MaterialFunctionCall("FadeSensitivity (optional)"=@n66) {materialFunction=&FadeBasedOnViewAngle}.Result
   )
   B=MaterialFunctionCall(Fade Length=ScalarParameter() {desc=Camera Fade distance., group="4) Camera Fade", parameterName=Camera Fade Distance, sortPriority=4}) {materialFunction=&CameraDepthFade}.Result
  )
  StaticSwitchParameter(
   True=Multiply(
    A=MaterialFunctionCall(Axis=@n62.RGB, Exponent=@n63) {materialFunction=&AxisAlignedFresnel}.Fresnel
    B=@n67
   )
   False=@n67
  ) {parameterName=UseAxisAlignedFresnel?}
 ]
}
