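# Translucent, unlit, two-sided Niagara master material for wind ribbon effects on mesh and sprite particles.
# Color blends a three-color UV gradient with screen-aligned noise layers crossfaded over time and added dust, tinting back faces separately.
# Opacity combines a ribbon mask, hardened texture dissolve, inverse Fresnel, depth and near-camera fades, and switchable backface culling.
# Particle color and dynamic parameters drive dissolve, desaturation, brightness and hue.
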Material /Game/Effects/Niagara/MasterMaterials/Wind/M_Ribbon
graph {
 properties {
  allowTranslucentCustomDepthWrites=true
  bUsedWithNiagaraMeshParticles=true
  bUsedWithNiagaraSprites=true
  blendMode=BLEND_Translucent
  pixelDepthOffsetMode=PDOM_Legacy
  shadingModel=MSM_Unlit
  twoSided=true
  usedShadingModels=Unlit
 }
 nodes [
  n1 = DynamicParameter() {defaultValue=<rgba 0 1 1 1>, paramNames [Dissolve,Desat,Scale,hue]}
  n2 = ParticleColor()
  n3 = VectorParameter() {parameterName=Color_CenterColor, defaultValue=<rgba 0.25 0.1485 0.0898 1>}
  n4 = VectorParameter() {parameterName=Color_ButtomColor, defaultValue=<rgba 0.16 0.0944 0.0617 1>}
  n5 = MaterialFunctionCall(Float2=TextureCoordinate()) {materialFunction=&BreakOutFloat2Components}
  n6 = Constant() {r=1}
  n7 = Constant()
  n8 = If(A=@n5.G, B=ScalarParameter() {defaultValue=0.7, parameterName=Color_ButtomPos}, "A > B"=@n6, "A == B"=@n6, "A < B"=@n7)
  n9 = ScalarParameter() {defaultValue=0.1, parameterName=Color_TopPos}
  n10 = ScalarParameter() {defaultValue=0.65, group=35.TextureBlend, parameterName=Trd_RemapLow}
  n11 = Time() {bOverride_Period=true, period=9999}
  n12 = Multiply(A=@n11, B=ScalarParameter() {defaultValue=0.5, group=35.TextureBlend, parameterName=Trd_ExtraLayerBlendSpeed})
  n13 = LinearInterpolate(A=NamedRerouteUsage(), B=NamedRerouteUsage(), Alpha=ConstantBiasScale(None=Sine(None=@n12)))
  n14 = TextureObject() {texture=/Game/Global/Textures/Default/PointSplatter/T_Tiling_Dust}
  n15 = VectorParameter() {parameterName=LightSand_UV}
  n16 = NamedRerouteUsage()
  n17 = Multiply(
   A=Power(
    Base=MaterialFunctionCall(Texture=@n14, UVControl_RGB=@n15.RGB, UVControl_A=@n15.RGB, "[Override]TimeInput"=@n11, "[Override]UVInput"=@n16) {materialFunction=&FXCommon_TextureSample}.R
    Exp=ScalarParameter() {defaultValue=1, parameterName=LightSand_Power}
   )
   B=VectorParameter() {parameterName=L_Color, defaultValue=<rgba 0 0 0 1>}.RGB
  )
  n18 = Multiply(
   A=LinearInterpolate(
    A=VectorParameter() {parameterName=Color_TopColor, defaultValue=<rgba 0.3 0.1783 0.1078 1>}.RGB
    B=LinearInterpolate(A=@n3.RGB, B=@n4.RGB, Alpha=@n8)
    Alpha=If(A=@n5.G, B=@n9, "A > B"=@n6, "A == B"=@n6, "A < B"=@n7)
   )
   B=Add(A=LinearInterpolate(A=@n10, Alpha=@n13), B=Multiply(A=@n13, B=@n17))
  )
  n19 = VectorParameter() {parameterName=InsideColorScale, defaultValue=<rgba 0.8 0.8 0.8 0>}
  n20 = Desaturation(None=Multiply(A=@n18, B=@n19.RGB), Fraction=@n19.A)
  n21 = Saturate(None=TwoSidedSign())
  n22 = MaterialFunctionCall(Hue Shift Percentage=@n1.hue, Texture=Multiply(A=@n2.RGB, B=LinearInterpolate(A=@n20, B=@n18, Alpha=@n21))) {materialFunction=&HueShift}
  MaterialFunctionCall(Emissive=Multiply(A=Desaturation(None=@n22.Result, Fraction=@n1.Desat), B=@n1.Scale)) {materialFunction=&MF_FX_EmissiveAdjustments}
  n23 = OneMinus(None=Fresnel(ExponentIn=ScalarParameter() {defaultValue=1, parameterName=Fresnel_Exp}, BaseReflectFractionIn=Constant()))
  n24 = ScalarParameter() {defaultValue=1, parameterName=Fresnel_Exp2}
  n25 = TextureObjectParameter() {parameterName=DissolveTexture, texture=/Game/CharClasses/Outlander/Gadgets/AntiMaterialCharge/Materials/T_AntiMaterialCharge_Dissolve_01}
  n26 = VectorParameter() {parameterName=UVControl, defaultValue=<rgba 1 0.5 0.2 0>}
  n27 = MaterialFunctionCall(Texture=@n25, UVControl_RGB=@n26.RGB, UVControl_A=@n26.RGB, "[Override]TimeInput"=@n11) {materialFunction=&FXCommon_TextureSample}
  n28 = DynamicParameter() {defaultValue=<rgba 0 1 1 1>, paramNames [Dissolve,Desat,Scale,hue]}
  n29 = Multiply(A=@n28.Dissolve) {constB=2}
  n30 = ScalarParameter() {parameterName=DissolveTest}
  n31 = Saturate(None=Subtract(A=@n27.R, B=Add(A=@n29, B=@n30)))
  n32 = ScalarParameter() {defaultValue=100, parameterName=DissolveHardness}
  n33 = VectorParameter() {parameterName=ScaleByCenter, defaultValue=<rgba 0.8 1.5 0 0>}
  n34 = TextureSampleParameter2D(UVs=MaterialFunctionCall(Texture Scale=AppendVector(A=@n33.R, B=@n33.R)) {materialFunction=&ScaleUVsByCenter}.UVs) {
   parameterName=MainTexture
   samplerType=SAMPLERTYPE_LinearGrayscale
   texture=/Game/Gadgets/Assets/VinderTech_GliderChute/Glider_Ensemble_Maroon_Male/FX/Textures/T_RibbonMask_D
  }
  n35 = VectorParameter() {parameterName=MainTextureRefine, defaultValue=<rgba 1 1 1 1>}
  n36 = MaterialFunctionCall(Texture=@n34.R, RGB=@n35.RGB, A=@n35.RGB) {materialFunction=&MF_FX_Refine}
  n37 = Multiply(A=Saturate(None=Multiply(A=@n31, B=@n32)), B=@n36.Result)
  n38 = Multiply(A=@n37, B=ScalarParameter() {defaultValue=0.8, parameterName=InsideOpacityScale})
  n39 = Multiply(A=@n2.A, B=Multiply(A=Power(Base=@n23, Exp=@n24), B=LinearInterpolate(A=@n38, B=@n37, Alpha=@n21)))
  n40 = Multiply(
   A=@n39
   B=MaterialFunctionCall!(ObjectOrParticlePosition=ParticlePositionWS()) {materialFunction=/Game/Effects/Fort_Effects/Materials/Toon/CustomDepthOutLine/MF_BackfaceCulling}.Result
  )
  Multiply(
   A=DepthFade(Opacity=StaticSwitchParameter(True=@n40, False=@n39) {parameterName=UseBackFaceCulling}, FadeDistance=ScalarParameter() {parameterName=DF})
   B=MaterialFunctionCall(Distance=ScalarParameter() {defaultValue=250, parameterName=CameraDistanceFade}) {materialFunction=&Near_Camera_Fade}.Result
  )
  n41 = VectorParameter() {group=32.TextureTex_Trd, parameterName=TextureTex_Trd_UVControl_ExtraLayer, defaultValue=<rgba 0.015 0.015 0 0>}
  n42 = Transform(None=Constant3Vector() {constant=<rgba 0 0 -1 0>}) {transformSourceType=TRANSFORMSOURCE_Camera}
  n43 = Transform(None=Constant3Vector() {constant=<rgba 1 0 0 0>}) {transformSourceType=TRANSFORMSOURCE_Camera}
  n44 = ObjectPositionWS()
  n45 = NamedRerouteDeclaration(None=MaterialFunctionCall(
    UVs=MaterialFunctionCall(Plane Normal Axis=@n42, Plane X Axis=@n43, Plane Center=@n44) {materialFunction=&VirtualPlaneCoordinates}.UVs
    Texture Scale=ScalarParameter() {defaultValue=1, group=00.Base, parameterName=Base_TextureTexUV}
   ) {materialFunction=&ScaleUVsByCenter}.UVs) {name=ScreenUV, nodeColor=<rgba 0 0.3203 1 1>}
  n46 = TextureObjectParameter() {
   group=32.TextureTex_Trd
   parameterName=TextureTex_Trd_Tex
   samplerType=SAMPLERTYPE_LinearGrayscale
   texture=/Game/Global/Textures/Default/Noise/T_WhiteNoise
  }
  NamedRerouteDeclaration(None=TextureSample!(
    UVs=MaterialFunctionCall!(UVInput=@n45, TimeInput=@n11, Control_RGB=@n41.RGB, Control_A=@n41.RGB) {materialFunction=/Game/Effects/Niagara/MasterMaterials/Functions/Common/FXCommon_QuickUVControl}.UV
    Tex=@n46
    Level
    Apply View MipBias
   ) {constMipValue=0, mipValueMode=TMVM_MipLevel, samplerType=SAMPLERTYPE_LinearColor}.R) {name=TextureTex_Trd_ExtraLayer, nodeColor=<rgba 0.3125 1 0 1>}
  n47 = VectorParameter() {group=32.TextureTex_Trd, parameterName=TextureTex_Trd_UVControl, defaultValue=<rgba 0.013 0.013 0 0>}
  NamedRerouteDeclaration(None=TextureSample(
    UVs=MaterialFunctionCall(UVInput=@n45, TimeInput=@n11, Control_RGB=@n47.RGB, Control_A=@n47.RGB) {materialFunction=&FXCommon_QuickUVControl}.UV
    Tex=@n46
   ) {constMipValue=0, mipValueMode=TMVM_MipLevel, samplerType=SAMPLERTYPE_LinearColor}.R) {name=TextureTex_Trd, nodeColor=<rgba 0.3125 1 0 1>}
  n48 = Subtract(A=@n27.R, B=Add(A=@n28.Dissolve, B=@n30))
  n49 = Constant()
  n50 = Constant() {r=1}
  n51 = Constant()
  Reroute(None=Multiply(A=If(A=@n48, B=@n49, "A > B"=@n50, "A == B"=@n50, "A < B"=@n51), B=@n36.Result))
 ]
}
