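# Niagara system for a wind cone burst effect, built from one mesh-rendered emitter that spawns short-lived particles in an instantaneous burst.
# The emitter takes its loop duration from the uDuration user parameter and spawns each mesh with a fixed rotation.
# Over each particle's life, curves scale the mesh up quickly from zero, easing out, while alpha spikes early and fades away.
# Dynamic material parameters and a per-particle material random value feed the mesh material.
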NiagaraSystem /Game/Effects/Niagara/MasterMaterials/Wind/NS_WindCone_Burst
dependencies {usedDataInterfaces [&NiagaraDataInterfaceCurve], usedDynamicInputs [], usedModules [&SystemState]}
emitters {WindCone_Burst {
  emitterUpdateScript [
   &EmitterState(
    Inactive Response=::NewEnumerator1
    Life Cycle Mode=::NewEnumerator1
    Loop Behavior=::NewEnumerator1
    Loop Duration=$User.uDuration
    UseLoopDelay=false
   )
   &SpawnBurst_Instantaneous(Use Spawn Probability=false, UseLoopCountLimit=false)
  ]
  particleSpawnScript [
   &InitializeParticle(Color Mode=::NewEnumerator0, Lifetime=0.4, Mesh Scale Mode=::NewEnumerator4, Position Mode=::NewEnumerator0, Write Material Random=true)
   &InitialMeshOrientation(Mesh Orientation Mode=::NewEnumerator2, Rotation=<0 0.25 0>, Use Rotation Vector=true)
  ]
  particleUpdateScript [
   &ParticleState()
   &ScaleMeshSize(Scale Factor=&VectorFromFloat(Value=&FloatFromCurve(FloatCurve=DataInterface() {propertyValues {Curve {keys [
          {arriveTangent=4.2179, interpMode=RCIM_Cubic, leaveTangent=4.2179, tangentMode=RCTM_User, time=0, value=0}
          {arriveTangent=0.8127, interpMode=RCIM_Cubic, leaveTangent=0.8127, tangentMode=RCTM_User, time=0.2494, value=0.7408}
          <at 1 1 RCIM_Cubic>
         ]}}})))
   &DynamicMaterialParameters(
    Index 0 Param 1=1
    Index 0 Param 2=1
    Index 0 Param 4=1
    Module.Index0Param0Write=true
    Module.Index0Param1Write=true
    Module.Index0Param2Write=true
    Module.Index0Param3Write=true
   )
   &ScaleColor(
    Scale Alpha=&FloatFromCurve(FloatCurve=DataInterface() {propertyValues {Curve {keys [
         <at 0 0 RCIM_Cubic>
         {arriveTangent=-1.2369, interpMode=RCIM_Cubic, leaveTangent=-1.2369, tangentMode=RCTM_User, time=0.1, value=0.8}
         <at 1 0 RCIM_Cubic>
        ]}}})
    ScaleA=true
    ScaleRGB=true
   )
  ]
  rendererClasses [&NiagaraMeshRendererProperties]
  renderers [{rendererClass=&NiagaraMeshRendererProperties, rendererIndex=0}]
 }}
summary {
 systemName=NS_WindCone_Burst
 emitters [{emitterName=WindCone_Burst, rendererClasses [&NiagaraMeshRendererProperties]}]
 userVariables [User.uDuration {type=NiagaraFloat, defaultValue=1.5, description=""}]
}
